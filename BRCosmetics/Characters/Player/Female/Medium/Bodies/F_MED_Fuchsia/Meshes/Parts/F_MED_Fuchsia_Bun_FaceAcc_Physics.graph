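# Physics asset for the bun hair of the female medium Fuchsia character's face accessory mesh, providing secondary motion for the bun and its loose strands.
# Capsule bodies cover the head, the bun and four strands, three of which have two segments.
# Constraints chain each strand to the bun and the bun to the head.
# All joints have tight swing and twist limits, tightest at the bun, so movement stays subtle.
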PhysicsAsset /BRCosmetics/Characters/Player/Female/Medium/Bodies/F_MED_Fuchsia/Meshes/Parts/F_MED_Fuchsia_Bun_FaceAcc_Physics
bodies {
 dyn_bun [{shapeType=Capsule, length=5.3723, radius=2.8746, shapeName=dyn_bun_capsule, rotation=<rot 0 -0.0008 -99.9998>}]
 dyn_bun_strand_01_01 [{
   shapeType=Capsule
   length=1.1645
   radius=0.5497
   shapeName=dyn_bun_strand_01_01_capsule
   center=<0.4057 0 0>
   rotation=<rot 90 -179.995 -179.9953>
  }]
 dyn_bun_strand_01_02 [{shapeType=Capsule, length=1.184, radius=0.5777, shapeName=dyn_bun_strand_01_02_capsule, center=<-0.1491 0 0>, rotation=<rot 90 0 -0.0005>}]
 dyn_bun_strand_02_01 [{
   shapeType=Capsule
   length=1.0682
   radius=0.6522
   shapeName=dyn_bun_strand_01_01_capsule
   center=<0.5012 0 0>
   rotation=<rot 90 -179.7468 -179.747>
  }]
 dyn_bun_strand_02_02 [{shapeType=Capsule, length=1.7362, radius=0.6453, shapeName=dyn_bun_strand_01_02_capsule, center=<0.0485 0 0>, rotation=<rot 90 0 -0.0005>}]
 dyn_bun_strand_03_01 [{
   shapeType=Capsule
   length=0.8602
   radius=0.5
   shapeName=dyn_bun_strand_01_01_capsule
   center=<0.4558 0 0>
   rotation=<rot 90 -160.5288 -160.5291>
  }]
 dyn_bun_strand_04_01 [{
   shapeType=Capsule
   length=0.7556
   radius=0.623
   shapeName=dyn_bun_strand_01_01_capsule
   center=<0.7891 0 0>
   rotation=<rot 90 -179.9901 -179.9903>
  }]
 dyn_bun_strand_04_02 [{shapeType=Capsule, length=1.0427, radius=0.6221, shapeName=dyn_bun_strand_01_01_capsule, center=<0.3089 0 0>, rotation=<rot 90 -90 -90.0003>}]
 head [{shapeType=Capsule, length=3.9899, radius=3.678, shapeName=head_capsule, center=<-0.4016 0 0>, rotation=<rot 90 179.8417 179.8421>}]
}
constraints [
 {bone1Name=dyn_bun_strand_01_02, bone2Name=dyn_bun_strand_01_01, swing1LimitDegrees=7.5, swing2LimitDegrees=7.5, twistLimitDegrees=2}
 {bone1Name=dyn_bun_strand_02_02, bone2Name=dyn_bun_strand_02_01, swing1LimitDegrees=7.5, swing2LimitDegrees=7.5, twistLimitDegrees=2}
 {bone1Name=dyn_bun_strand_04_02, bone2Name=dyn_bun_strand_04_01, swing1LimitDegrees=7.5, swing2LimitDegrees=7.5, twistLimitDegrees=2}
 {bone1Name=dyn_bun, bone2Name=head, swing1LimitDegrees=5, swing2LimitDegrees=5, twistLimitDegrees=2}
 {bone1Name=dyn_bun_strand_01_01, bone2Name=dyn_bun, swing1LimitDegrees=7.5, swing2LimitDegrees=7.5, twistLimitDegrees=2}
 {bone1Name=dyn_bun_strand_02_01, bone2Name=dyn_bun, swing1LimitDegrees=7.5, swing2LimitDegrees=7.5, twistLimitDegrees=2}
 {bone1Name=dyn_bun_strand_04_01, bone2Name=dyn_bun, swing1LimitDegrees=7.5, swing2LimitDegrees=7.5, twistLimitDegrees=2}
 {bone1Name=dyn_bun_strand_03_01, bone2Name=dyn_bun, swing1LimitDegrees=7.5, swing2LimitDegrees=7.5, twistLimitDegrees=2}
]
