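# Physics asset for the face accessory part of the female medium Fuchsia player cosmetic, setting up physics for its left and right hair bangs.
# Each side is a four-bone chain hanging from the head, with paired capsules per bone and tight swing and twist limits.
# Capsules on the head, neck, upper spine, clavicles and upper arms approximate the body around the bangs.
# Of those body bones, only the right clavicle and upper arm have constraints, with free angular motion.
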PhysicsAsset /BRCosmetics/Characters/Player/Female/Medium/Bodies/F_MED_Fuchsia/Meshes/Parts/F_MED_Fuchsia_FaceAcc_Physics
bodies {
 clavicle_l [{shapeType=Capsule, length=8.7782, radius=5.8551, shapeName=clavicle_l_capsule, center=<4.6786 0.123 0.1272>, rotation=<rot 90 -90 -90.0004>}]
 clavicle_r [{
   shapeType=Capsule
   length=8.7782
   radius=5.8551
   shapeName=clavicle_l_capsule
   center=<-4.6786 -0.123 -0.1272>
   rotation=<rot 90 -160.5289 19.4707>
  }]
 dyn_bangs_01_l [
  {
   shapeType=Capsule
   length=5.269
   radius=1.1912
   shapeName=dyn_bangs_04_l_capsule
   center=<3.3301 -0.6102 0.9537>
   rotation=<rot -90 0 -0.0001>
  }
  {
   shapeType=Capsule
   length=5.269
   radius=1.1912
   shapeName=dyn_bangs_04_l_capsule
   center=<3.3467 -0.8175 -1.3323>
   rotation=<rot -85.0001 179.9995 -179.9997>
  }
 ]
 dyn_bangs_01_r [
  {
   shapeType=Capsule
   length=5.683
   radius=0.7773
   shapeName=dyn_bangs_04_l_capsule
   center=<3.2188 -0.5054 -0.9776>
   rotation=<rot -90 -162.5816 167.5815>
  }
  {
   shapeType=Capsule
   length=5.683
   radius=0.7773
   shapeName=dyn_bangs_04_l_capsule
   center=<3.2188 -0.5054 1.3328>
   rotation=<rot -90 -162.5816 167.5815>
  }
 ]
 dyn_bangs_02_l [
  {
   shapeType=Capsule
   length=5.291
   radius=1.1692
   shapeName=dyn_bangs_04_l_capsule
   center=<3.3301 -0.4512 0.4722>
   rotation=<rot -90 0 -0.0001>
  }
  {
   shapeType=Capsule
   length=5.291
   radius=1.1692
   shapeName=dyn_bangs_04_l_capsule
   center=<3.3301 -0.4512 -1.5653>
   rotation=<rot -90 0 -0.0001>
  }
 ]
 dyn_bangs_02_r [
  {
   shapeType=Capsule
   length=5.683
   radius=0.7773
   shapeName=dyn_bangs_04_l_capsule
   center=<3.2188 -0.1731 -0.9776>
   rotation=<rot -90 0 -0.0001>
  }
  {
   shapeType=Capsule
   length=5.683
   radius=0.7773
   shapeName=dyn_bangs_04_l_capsule
   center=<3.2188 -0.1731 1.348>
   rotation=<rot -84.9999 179.9999 180>
  }
 ]
 dyn_bangs_03_l [
  {
   shapeType=Capsule
   length=5.4016
   radius=1.0586
   shapeName=dyn_bangs_04_l_capsule
   center=<3.3522 -0.4273 -1.3262>
   rotation=<rot -79.9998 -0.0003 0.0001>
  }
  {
   shapeType=Capsule
   length=5.4016
   radius=1.0586
   shapeName=dyn_bangs_04_l_capsule
   center=<3.3301 -0.4273 0.3629>
   rotation=<rot -90 0 -0.0001>
  }
 ]
 dyn_bangs_03_r [
  {
   shapeType=Capsule
   length=5.683
   radius=0.7773
   shapeName=dyn_bangs_04_l_capsule
   center=<3.3256 -0.2523 -0.7264>
   rotation=<rot -85.0002 -0.0001 -0.0002>
  }
  {
   shapeType=Capsule
   length=5.683
   radius=0.7773
   shapeName=dyn_bangs_04_l_capsule
   center=<3.1214 -0.2523 0.9014>
   rotation=<rot -84.9999 179.9997 -179.9999>
  }
 ]
 dyn_bangs_04_l [
  {
   shapeType=Capsule
   length=6.2202
   radius=0.7862
   shapeName=dyn_bangs_04_l_capsule
   center=<3.6956 -0.2138 0.5911>
   rotation=<rot -90 0 -0.0001>
  }
  {
   shapeType=Capsule
   length=4.5742
   radius=0.7862
   shapeName=dyn_bangs_04_l_capsule
   center=<2.8335 -0.2138 -0.4323>
   rotation=<rot -84.9996 -0.0011 0.0008>
  }
 ]
 dyn_bangs_04_r [
  {
   shapeType=Capsule
   length=5.683
   radius=0.7773
   shapeName=dyn_bangs_04_l_capsule
   center=<3.3109 -0.2523 -0.2243>
   rotation=<rot -85.0001 0 -0.0002>
  }
  {shapeType=Capsule, length=3.7635, radius=0.5, shapeName=dyn_bangs_04_l_capsule, center=<2.1377 -0.2523 0.8487>, rotation=<rot -90 0 -0.0001>}
 ]
 head [
  {shapeType=Capsule, length=4.0218, radius=6.7511, shapeName=head_capsule, center=<2.2998 1.7262 0>, rotation=<rot 90 180 -172.1549>}
  {shapeType=Capsule, length=6.9367, radius=3.384, shapeName=head_capsule, center=<-0.8661 4.6175 0>, rotation=<rot -90 20.556 -32.8634>}
 ]
 neck_01 [
  {shapeType=Capsule, length=1.3026, radius=5.3705, shapeName=neck_02_capsule, center=<-1.6291 0 0>, rotation=<rot 90 0.0002 0.0007>}
  {
   shapeType=Capsule
   length=13.4033
   radius=5.8551
   shapeName=clavicle_l_capsule
   center=<2.3801 -5.3812 -10.9431>
   rotation=<rot -1.7414 -20.2389 -13.7862>
  }
  {
   shapeType=Capsule
   length=13.4033
   radius=5.8551
   shapeName=clavicle_l_capsule
   center=<2.3801 -5.3812 10.9431>
   rotation=<rot 1.7414 -20.2389 13.7862>
  }
 ]
 spine_04 [{shapeType=Capsule, length=6.8666, radius=8.4903, shapeName=spine_04_capsule, center=<1.0106 2.9769 0>}]
 spine_05 [{shapeType=Capsule, length=12.2047, radius=8.2651, shapeName=spine_05_capsule, center=<-0.4634 0 0>}]
 upperarm_l [{
   shapeType=Capsule
   length=25.957
   radius=4.043
   shapeName=upperarm_l_capsule
   center=<13.7803 0.0001 -0.8185>
   rotation=<rot 89.6 -179.9927 -179.993>
  }]
 upperarm_r [{
   shapeType=Capsule
   length=25.957
   radius=4.043
   shapeName=upperarm_l_capsule
   center=<-13.7803 -0.0001 0.8185>
   rotation=<rot 89.6 -179.9927 0.007>
  }]
}
constraints [
 {bone1Name=dyn_bangs_01_l, bone2Name=head, swing1LimitDegrees=10, swing2LimitDegrees=15, twistLimitDegrees=5}
 {bone1Name=dyn_bangs_02_l, bone2Name=dyn_bangs_01_l, swing1LimitDegrees=15, swing2LimitDegrees=15, twistLimitDegrees=5}
 {bone1Name=dyn_bangs_03_l, bone2Name=dyn_bangs_02_l, swing1LimitDegrees=20, swing2LimitDegrees=15, twistLimitDegrees=5}
 {bone1Name=dyn_bangs_04_l, bone2Name=dyn_bangs_03_l, swing1LimitDegrees=10, swing2LimitDegrees=15, twistLimitDegrees=5}
 {bone1Name=dyn_bangs_04_r, bone2Name=dyn_bangs_03_r, swing1LimitDegrees=10, swing2LimitDegrees=15, twistLimitDegrees=5}
 {bone1Name=dyn_bangs_03_r, bone2Name=dyn_bangs_02_r, swing1LimitDegrees=20, swing2LimitDegrees=15, twistLimitDegrees=5}
 {bone1Name=dyn_bangs_02_r, bone2Name=dyn_bangs_01_r, swing1LimitDegrees=15, swing2LimitDegrees=15, twistLimitDegrees=5}
 {bone1Name=dyn_bangs_01_r, bone2Name=head, swing1LimitDegrees=10, swing2LimitDegrees=15, twistLimitDegrees=5}
 {
  bone1Name=upperarm_r
  bone2Name=clavicle_r
  swing1LimitDegrees=45
  swing1Motion=Free
  swing2LimitDegrees=45
  swing2Motion=Free
  twistLimitDegrees=45
  twistMotion=Free
 }
 {
  bone1Name=clavicle_r
  bone2Name=spine_05
  swing1LimitDegrees=45
  swing1Motion=Free
  swing2LimitDegrees=45
  swing2Motion=Free
  twistLimitDegrees=45
  twistMotion=Free
 }
]
