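# Animation Blueprint for the Fuchsia female character's face accessory, adding secondary motion to hair and mask.
# It copies the parent mesh pose, springs the bun and strands with AnimDynamics, and additively rotates the bangs while skydiving.
# Two head-space RigidBody simulations follow, selected by mask-off or mask-on state from the Stage1 cosmetic variant tag when customization loads.
# A deprecated state function updates their gravity override each frame.
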AnimBlueprint /BRCosmetics/Characters/Player/Female/Medium/Bodies/F_MED_Fuchsia/Meshes/Parts/F_MED_Fuchsia_FaceAcc_AnimBP
target_skeleton=&/F_MED_Fuchsia_FaceAcc_Skeleton
graphs [
 AnimGraph {nodes [
   RigidBody!(ComponentPose, Alpha, OverrideWorldGravity:Vector=Get gravity_override() -> (gravity_override:Vector), ExternalForce) -> (Pose) {
    bOverrideWorldGravity=true
    overridePhysicsAsset=&/F_MED_Fuchsia_Bun_FaceAcc_Physics
    simulationSpace=BaseBoneSpace
    alphaBoolBlend {blendOption=Linear}
    baseBoneRef {boneName=head}
    componentAppliedLinearAccClamp=<3000 3000 3000>
    componentLinearAccScale=<0.4 0.3 0.3>
    componentLinearVelScale=<0.4 0.3 0.3>
    overrideWorldGravity {z=-980}
    simSpaceSettings {maxAngularAcceleration=1000, maxAngularVelocity=3, maxLinearAcceleration=10, maxLinearVelocity=100, velocityScaleZ=0, worldAlpha=1}
   }
   n1 = Get gravity_override()
   n2 = RigidBody(
    ComponentPose=Local To Component(LocalPose=Use cached pose 'skydive_pose_cache'())
    Alpha=Get mask_off_alpha() -> (mask_off_alpha:double)
    OverrideWorldGravity=@n1
   ) {
    bOverrideWorldGravity=true
    overridePhysicsAsset=&/F_MED_Fuchsia_FaceAcc_Physics
    simulationSpace=BaseBoneSpace
    alphaBoolBlend {blendOption=Linear}
    baseBoneRef {boneName=head}
    componentAppliedLinearAccClamp=<3000 3000 3000>
    componentLinearAccScale=<0.4 0.3 0.3>
    componentLinearVelScale=<0.4 0.3 0.3>
    overrideWorldGravity {z=-980}
    simSpaceSettings {maxAngularAcceleration=1000, maxAngularVelocity=3, maxLinearAcceleration=250, maxLinearVelocity=500, velocityScaleZ=0, worldAlpha=1}
   }
   n3 = RigidBody(
    ComponentPose=Local To Component(LocalPose=Use cached pose 'skydive_pose_cache'())
    Alpha=Get mask_on_alpha() -> (mask_on_alpha:double)
    OverrideWorldGravity=@n1
   ) {
    bOverrideWorldGravity=true
    overridePhysicsAsset=&/F_MED_Fuchsia_Mask_FaceAcc_Physics
    simulationSpace=BaseBoneSpace
    alphaBoolBlend {blendOption=Linear}
    baseBoneRef {boneName=head}
    componentAppliedLinearAccClamp=<3000 3000 3000>
    componentLinearAccScale=<0.4 0.3 0.3>
    componentLinearVelScale=<0.4 0.3 0.3>
    overrideWorldGravity {z=-980}
    simSpaceSettings {maxAngularAcceleration=1000, maxAngularVelocity=3, maxLinearAcceleration=250, maxLinearVelocity=500, velocityScaleZ=0, worldAlpha=1}
   }
   Output Pose(Result=Blend Poses by bool(
     BlendPose_0=Component To Local(ComponentPose=@n2)
     BlendPose_1=Component To Local(ComponentPose=@n3)
     bActiveValue=Get mask_off() -> (mask_off:bool)
    ) {blendType=Linear}) {layerGroup=None}
   n4 = Anim Dynamics - Bone: dyn_bun(ComponentPose=Local To Component(LocalPose=Copy Pose From Mesh() {bUseAttachedParent=true}), Alpha=1, ExternalForce="0,0,0") {
    angularDampingOverride=1
    angularSpringConstant=15000
    bAngularSpring=true
    bOverrideAngularDamping=true
    bUseGravityOverride=true
    linearSpringConstant=35000
    simulationSpace=BoneRelative
    alphaBoolBlend {blendOption=Linear}
    boundBone {boneName=dyn_bun}
    componentAppliedLinearAccClamp=<3000 3000 3000>
    componentLinearAccScale=<0.1 0.1 0.1>
    componentLinearVelScale=<0.9 0.9 0.9>
    gravityOverride {z=-980}
    physicsBodyDefinitions [{
      boundBone {boneName=dyn_bun}
      boxExtents=<40 40 40>
      constraintSetup {angularLimitsMax=<1 2 2>, angularLimitsMin=<-1 -2 -2>, angularTarget=<1 0 0>}
      localJointOffset=<0.15 0 0>
     }]
    relativeSpaceBone {boneName=pelvis}
   }
   n5 = Anim Dynamics - Bone: dyn_bun_strand_01_01(ComponentPose=@n4, Alpha=1, ExternalForce="0,0,0") {
    angularDampingOverride=1
    angularSpringConstant=7500
    bAngularSpring=true
    bOverrideAngularDamping=true
    bUseGravityOverride=true
    linearSpringConstant=7500
    simulationSpace=BoneRelative
    alphaBoolBlend {blendOption=Linear}
    boundBone {boneName=dyn_bun_strand_01_01}
    componentAppliedLinearAccClamp=<3000 3000 3000>
    componentLinearAccScale=<0.1 0.1 0.1>
    componentLinearVelScale=<0.9 0.9 0.9>
    gravityOverride {z=-980}
    physicsBodyDefinitions [{
      boundBone {boneName=dyn_bun_strand_01_01}
      boxExtents=<25 25 25>
      constraintSetup {angularLimitsMax=<1 1 10>, angularLimitsMin=<-1 -1 -10>, angularTarget=<1 0 0>}
      localJointOffset=<0.75 0 0>
     }]
    relativeSpaceBone {boneName=pelvis}
   }
   n6 = Anim Dynamics - Bone: dyn_bun_strand_02_01(ComponentPose=@n5, Alpha=1, ExternalForce="0,0,0") {
    angularDampingOverride=1
    angularSpringConstant=7250
    bAngularSpring=true
    bOverrideAngularDamping=true
    bUseGravityOverride=true
    linearSpringConstant=7500
    simulationSpace=BoneRelative
    alphaBoolBlend {blendOption=Linear}
    boundBone {boneName=dyn_bun_strand_02_01}
    componentAppliedLinearAccClamp=<3000 3000 3000>
    componentLinearAccScale=<0.1 0.1 0.1>
    componentLinearVelScale=<0.9 0.9 0.9>
    gravityOverride {z=-980}
    physicsBodyDefinitions [{
      boundBone {boneName=dyn_bun_strand_02_01}
      boxExtents=<22.5 22.5 22.5>
      constraintSetup {angularLimitsMax=<1 1 10>, angularLimitsMin=<-1 -1 -10>, angularTarget=<1 0 0>}
      localJointOffset=<0.65 0 0>
     }]
    relativeSpaceBone {boneName=pelvis}
   }
   n7 = Anim Dynamics - Bone: dyn_bun_strand_03_01(ComponentPose=@n6, Alpha=1, ExternalForce="0,0,0") {
    angularDampingOverride=1
    angularSpringConstant=7250
    bAngularSpring=true
    bOverrideAngularDamping=true
    bUseGravityOverride=true
    linearSpringConstant=7500
    simulationSpace=BoneRelative
    alphaBoolBlend {blendOption=Linear}
    boundBone {boneName=dyn_bun_strand_03_01}
    componentAppliedLinearAccClamp=<3000 3000 3000>
    componentLinearAccScale=<0.1 0.1 0.1>
    componentLinearVelScale=<0.9 0.9 0.9>
    gravityOverride {z=-980}
    physicsBodyDefinitions [{
      boundBone {boneName=dyn_bun_strand_03_01}
      boxExtents=<21 21 21>
      constraintSetup {angularLimitsMax=<1 1 10>, angularLimitsMin=<-1 -1 -10>, angularTarget=<1 0 0>}
      localJointOffset=<0.65 0 0>
     }]
    relativeSpaceBone {boneName=pelvis}
   }
   n8 = Anim Dynamics - Bone: dyn_bun_strand_04_01(ComponentPose=@n7, Alpha=1, ExternalForce="0,0,0") {
    angularDampingOverride=1
    angularSpringConstant=7250
    bAngularSpring=true
    bOverrideAngularDamping=true
    bUseGravityOverride=true
    linearSpringConstant=7500
    simulationSpace=BoneRelative
    alphaBoolBlend {blendOption=Linear}
    boundBone {boneName=dyn_bun_strand_04_01}
    componentAppliedLinearAccClamp=<3000 3000 3000>
    componentLinearAccScale=<0.1 0.1 0.1>
    componentLinearVelScale=<0.9 0.9 0.9>
    gravityOverride {z=-980}
    physicsBodyDefinitions [{
      boundBone {boneName=dyn_bun_strand_04_01}
      boxExtents=<23 23 23>
      constraintSetup {angularLimitsMax=<1 1 10>, angularLimitsMin=<-1 -1 -10>, angularTarget=<1 0 0>}
      localJointOffset=<0.65 0 0>
     }]
    relativeSpaceBone {boneName=pelvis}
   }
   Save cached pose 'base_pose_cache'(Pose=Component To Local(ComponentPose=@n8)) {cacheName=base_pose_cache}
   n9 = "Transform (Modify) Bone - Bone: dyn_bangs_mstr_l"(
    ComponentPose=Local To Component(LocalPose=Use cached pose 'base_pose_cache'())
    Alpha=1
    Translation="0,0,0"
    Rotation="20,-10,25"
    Scale="1,1,1"
   ) {rotationMode=BMM_Additive, alphaBoolBlend {blendOption=Linear}, boneToModify {boneName=dyn_bangs_mstr_l}, rotation=<rot 20 -10 25>}
   n10 = "Transform (Modify) Bone - Bone: dyn_bangs_mstr_r"(ComponentPose=@n9, Alpha=1, Translation="0,0,0", Rotation="-10,-15,30", Scale="1,1,1") {rotationMode=BMM_Additive, alphaBoolBlend {blendOption=Linear}, boneToModify {boneName=dyn_bangs_mstr_r}, rotation=<rot -10 -15 30>}
   Save cached pose 'skydive_pose_cache'(Pose=Blend Poses by bool(
     BlendPose_0=Component To Local(ComponentPose=@n10)
     BlendPose_1=Use cached pose 'base_pose_cache'()
     BlendTime_0=0.25
     BlendTime_1=0.25
     bActiveValue=Get bIsSkyDiving()
    ) {blendType=Linear}) {cacheName=skydive_pose_cache}
  ]}
 EventGraph {nodes [
   Event BlueprintUpdateAnimation()
   > n1 = Gravity Override State Function DEPRECATED(CharacterPartAnimInstance=Self-Reference(), GameTimeInSeconds=GetGameTimeInSeconds())
   > Set gravity_override(gravity_override=@n1)
   Event OnCharacterCustomizationLoaded()
   > n2 = Branch(Condition=CharacterItemDefinitionHasAnyMetaTag(Tags=Make Array("[0]"=(TagName="Cosmetics.Variant.Property.Stage1"))))
   > Set mask_off(mask_off:bool=true) -> (Output_Get:bool)
   > Set mask_off_alpha(mask_off_alpha:double=1) -> (Output_Get:double)
   > Set mask_on_alpha(mask_on_alpha:double=0) -> (Output_Get:double)
   Set mask_off(execute=@n2.else, mask_off=false)
   > Set mask_off_alpha(mask_off_alpha=0)
   > Set mask_on_alpha(mask_on_alpha=1)
  ]}
]
variables [
 gravity_override {type=Vector}
 mask_off {type=boolean}
 mask_off_alpha {type=number}
 mask_on_alpha {type=number}
]
